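# Translucent unlit VFX material for a RoutineBarley-branded stylized water wave, animated through dynamic parameters.
# A custom node builds a traveling Gaussian crest with lean, curl and lifetime envelope, outputting offset and alpha.
# Screen derivatives of the displaced position rebuild normals for Fresnel, which shapes height-based, depth-faded opacity.
# Color layers Fresnel water tints, vertex-masked caustic foam, directional shading and a fake reflection highlight, distorted by panning noise.
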Material /DealershipSharedLibrary/Assets/ThirdPartyBranding/RoutineBarley/M_DLR_RoutineBarley_Wave
graph {
 properties {
  bEnableResponsiveAA=true
  bUsedWithMorphTargets=true
  bUsedWithSkeletalMesh=true
  blendMode=BLEND_Translucent
  maxWorldPositionOffsetDisplacement=1000
  opacityMaskClipValue=0.5893
  pixelDepthOffsetMode=PDOM_Legacy
  shadingModel=MSM_Unlit
  translucencyLightingMode=TLM_VolumetricPerVertexNonDirectional
  usedShadingModels=Unlit
 }
 nodes [
  n1 = VectorParameter() {parameterName=Param_2, defaultValue=<rgba -1 0 0 1>}
  n2 = TextureCoordinate() {vTiling=2}
  n3 = SphereMask(A=@n2, B=Constant() {r=0.5}, Radius=LinearInterpolate(Alpha=NamedRerouteUsage()) {constA=0.2}) {attenuationRadius=0.35, hardnessPercent=0.4}
  n4 = MaterialFunctionCall(In=@n3, Power=LinearInterpolate(Alpha=NamedRerouteUsage()) {constA=1.7, constB=1.3}) {materialFunction=&SCurve}
  n5 = Multiply(A=Multiply(A=ScalarParameter() {defaultValue=0.8167}, B=NamedRerouteUsage()), B=@n4.Result)
  n6 = Multiply(A=VectorParameter() {parameterName=Param_1, defaultValue=<rgba 0 0 1172.5425 1>}.RGB, B=NamedRerouteUsage())
  n7 = MaterialFunctionCall(In=@n3, Power=ScalarParameter() {defaultValue=0.729, parameterName=Param_1}) {materialFunction=&SCurve}
  n8 = VectorParameter() {defaultValue=<rgba 0 0 269.6088 1>}
  n9 = NamedRerouteUsage()
  Add(
   A=Multiply(A=RotateAboutAxis(NormalizedRotationAxis=@n1.RGB, RotationAngle=@n5, PivotPoint=@n6, Position=@n7.Result), B=@n4.Result)
   B=Multiply(A=@n4.Result, B=Multiply(A=@n8.RGB, B=@n9))
  )
  n10 = DynamicParameter()
  n11 = Constant() {r=1}
  Add(A=Multiply(A=@n2, B=AppendVector(A=@n10.Param1, B=@n11)), B=AppendVector(A=@n10.Param2, B=Constant()))
  ComponentMask(None=@n2) {g=true}
  MaterialExpressionWorldPosition_2 = WorldPosition()
  MaterialExpressionVertexTangentWS_0 = VertexTangentWS()
  MaterialExpressionConstant3Vector_8 = Constant3Vector()
  MaterialExpressionScalarParameter_41 = ScalarParameter() {parameterName=SampleOffset}
  Custom!(
   OriginalPos=@MaterialExpressionWorldPosition_2.XYZ
   TangentWS=@MaterialExpressionVertexTangentWS_0
   BitangentWS=@MaterialExpressionConstant3Vector_8.R
   SampleOffset=@MaterialExpressionScalarParameter_41
   ComputeWPO
  ) {
   code="// Three samples of the displaced surface\r\nfloat3 DisplacedCenter      = OriginalPos               + ComputeWPO(OriginalPos);\r\nfloat3 DisplacedAlongTangent   = (OriginalPos + TangentWS  * SampleOffset)\r\n                                + ComputeWPO(OriginalPos + TangentWS  * SampleOffset);\r\nfloat3 DisplacedAlongBitangent = (OriginalPos + BitangentWS * SampleOffset)\r\n                                + ComputeWPO(OriginalPos + BitangentWS * SampleOffset);\r\n\r\n// Two edges of the displaced surface, crossed = new normal\r\nfloat3 EdgeAlongTangent   = DisplacedAlongTangent   - DisplacedCenter;\r\nfloat3 EdgeAlongBitangent = DisplacedAlongBitangent - DisplacedCenter;\r\nfloat3 NormalWS = normalize(cross(EdgeAlongTangent, EdgeAlongBitangent));"
   showCode=true
   inputs [
    {
     inputName=OriginalPos
     input {
      expression=&self:MaterialExpressionWorldPosition_2
      mask=1
      maskB=1
      maskG=1
      maskR=1
     }
    }
    {
     inputName=TangentWS
     input {expression=&self:MaterialExpressionVertexTangentWS_0}
    }
    {
     inputName=BitangentWS
     input {
      expression=&self:MaterialExpressionConstant3Vector_8
      mask=1
      maskR=1
      outputIndex=1
     }
    }
    {
     inputName=SampleOffset
     input {expression=&self:MaterialExpressionScalarParameter_41}
    }
    {inputName=ComputeWPO, input {expression=null}}
   ]
  }
  Desaturation(None=Constant3Vector() {constant=<rgba 1000 1000 1000 0>}, Fraction=Constant())
  LocalPosition()
  n12 = TextureCoordinate()
  n13 = Multiply(A=TextureCoordinate(), B=ScalarParameter() {defaultValue=1.4658, parameterName=Color Variation Tile})
  n14 = TextureObjectParameter() {parameterName=Color Noise Texture, texture=/Game/Global/Textures/Default/Noise/T_SoftNoise}
  n15 = TextureSample(UVs=Panner(Coordinate=@n13, Speed=Constant2Vector() {g=0.8225, r=0.1}), Tex=@n14) {texture=&T_ShieldCaustics}
  n16 = Add(A=@n13, B=Constant2Vector() {g=0.3548, r=0.4181})
  n17 = Constant2Vector() {g=0.8676, r=0.1}
  n18 = TextureSample(UVs=Multiply(A=Panner(Coordinate=@n16, Speed=@n17)) {constB=0.6183}, Tex=@n14) {texture=&T_ShieldCaustics}
  n19 = Add(A=@n13, B=Constant2Vector() {g=0.1484, r=0.8649})
  n20 = Constant2Vector() {g=0.788, r=-0.1}
  n21 = TextureSample(UVs=Multiply(A=Panner(Coordinate=@n19, Speed=@n20)) {constB=0.8088}, Tex=@n14) {texture=&T_ShieldCaustics}
  n22 = Multiply(A=Multiply(A=@n15.R, B=@n18.G), B=@n21.B)
  n23 = Multiply(
   A=BumpOffset(
    Coordinate=Panner(Coordinate=@n12) {speedY=-1}
    Height=Multiply(A=@n22)
    HeightRatioInput=ScalarParameter() {defaultValue=0.2547, parameterName=Param_16}
   )
   B=AppendVector(
    A=ScalarParameter() {defaultValue=1.6543, parameterName=Caustic_Scale_X}
    B=ScalarParameter() {defaultValue=0.4517, parameterName=Caustic_Scale_Y}
   )
  )
  n24 = Multiply(A=@n23, B=ScalarParameter() {defaultValue=4.0259, parameterName=Param_7})
  n25 = Multiply(
   A=Add(A=TextureSample(UVs=@n24) {texture=/Game/Global/Textures/Default/Masks_and_Gradients/T_BlasterBeamDistort}.R) {constB=-0.5}
   B=ScalarParameter() {defaultValue=0.4488, parameterName=Param_17}
  )
  n26 = ReflectionVectorWS()
  n27 = VectorParameter() {parameterName=Param_6, defaultValue=<rgba 0.7927 0 1 1>}
  n28 = Multiply(
   A=SphereMask(
    A=Add(A=@n25, B=@n26)
    B=Transform(None=@n27.RGB) {transformSourceType=TRANSFORMSOURCE_Local}
    Radius=ScalarParameter() {defaultValue=0.4, parameterName=Param_4}
    Hardness=ScalarParameter() {defaultValue=0.2, parameterName=Param_9}
   )
   B=ScalarParameter() {defaultValue=4, parameterName=Param_12}
  ) {constB=2}
  n29 = ScalarParameter() {defaultValue=2.7258, parameterName=Param_11}
  n30 = Multiply(
   A=Multiply(
    A=TextureSample(UVs=@n23) {texture=&T_ShieldCaustics}.R
    B=ScalarParameter() {defaultValue=1, parameterName=Foam Amount}
   )
   B=Power(Base=VertexColor().R, Exp=ScalarParameter() {defaultValue=12, parameterName=Param_10})
  )
  n31 = ScalarParameter() {defaultValue=0.5, parameterName=Param_8}
  n32 = LinearInterpolate(
   A=LinearInterpolate(
    A=VectorParameter() {parameterName=Param_4, defaultValue=<rgba 0.0355 0.135 0.401 1>}.RGB
    B=VectorParameter() {parameterName=Param_7, defaultValue=<rgba 0.1236 0.3674 0.5365 1>}.RGB
    Alpha=Fresnel(ExponentIn=@n29) {baseReflectFraction=0, exponent=1}
   ) {constA=1, constB=0}
   B=VectorParameter() {parameterName=FoamColor, defaultValue=<rgba 0.9792 0.9792 0.9792 1>}.RGB
   Alpha=Saturate(None=Power(Base=@n30, Exp=@n31))
  )
  n33 = Multiply(A=@n32, B=VectorParameter() {parameterName=Color 2, defaultValue=<rgba 0.3383 0.4998 0.5156 1>}.RGB)
  n34 = VectorParameter() {parameterName=Param_5, defaultValue=<rgba -0.1867 0 1 1>}
  n35 = ConstantBiasScale(None=DotProduct(A=VertexNormalWS(), B=Transform(None=@n34.RGB) {transformSourceType=TRANSFORMSOURCE_Local}))
  n36 = DynamicParameter() {parameterIndex=1, paramNames [Shadow Exp, Param2, Param3, Param4]}
  n37 = MaterialFunctionCall(In=Add(A=Power(Base=@n35, Exp=@n36.Shadow Exp), B=@n25), Contrast=@n36.Shadow Exp) {materialFunction=&CheapContrast}
  MaterialFunctionCall(Emissive=Add(A=Multiply(A=@n28, B=@n28), B=LinearInterpolate(A=@n33, B=@n32, Alpha=@n37.Result))) {materialFunction=&MF_FX_EmissiveAdjustments}
  MaterialFunctionCall(
   Opacity In=ScalarParameter() {defaultValue=1, parameterName=Opacity}
   DepthFade Distance=ScalarParameter() {defaultValue=431.3175, parameterName=Param_15}
  ) {materialFunction=&MF_VCVFX_DepthFade}
  n38 = ParticleColor()
  n39 = LocalPosition()
  MaterialExpressionLocalPosition_1 = LocalPosition()
  MaterialExpressionDynamicParameter_6 = DynamicParameter() {parameterIndex=2, defaultValue=<rgba 0.2 1 1 1>, paramNames [Time, Wave Speed, Wave Hight, Wave Width]}
  n40 = VectorParameter() {parameterName=Param_3, defaultValue=<rgba 1 0 -50 0>}
  MaterialExpressionComponentMask_2 = ComponentMask(None=@n40.RGB) {g=true, r=true}
  MaterialExpressionAppendVector_2 = AppendVector(A=@n40.B, B=@n40.B)
  n41 = ComponentMask(None=TextureCoordinate() {vTiling=0.5}) {g=true}
  MaterialExpressionMultiply_38 = Multiply(
   A=@MaterialExpressionDynamicParameter_6.Wave Hight
   B=MaterialFunctionCall(In=Sine(None=@n41), Power=ScalarParameter() {defaultValue=1.2836, parameterName=s}) {materialFunction=&SCurve}.Result
  )
  MaterialExpressionDynamicParameter_7 = DynamicParameter() {parameterIndex=3, defaultValue=<rgba 0.2 1 1 1>, paramNames [Front Sharpness, CrestLean, CurlStrength, LifeWindow]}
  n42 = Custom!(
   LocalPos=@MaterialExpressionLocalPosition_1.XYZ
   LocalTime=@MaterialExpressionDynamicParameter_6.Time
   WaveDir=@MaterialExpressionComponentMask_2
   WaveOrigin=@MaterialExpressionAppendVector_2
   WaveSpeed=@MaterialExpressionDynamicParameter_6.Time
   WaveHeight=@MaterialExpressionMultiply_38
   WaveWidth=@MaterialExpressionDynamicParameter_6.Time
   FrontSharpness=@MaterialExpressionDynamicParameter_7.Front Sharpness
   CrestLean=@MaterialExpressionDynamicParameter_7.Front Sharpness
   CurlStrength=@MaterialExpressionDynamicParameter_7.Front Sharpness
   LifeWindow=@MaterialExpressionDynamicParameter_7.Front Sharpness
  ) {
   code="float2 travelDir = normalize(WaveDir + float2(1e-5, 0));\r\nfloat2 relPos    = LocalPos.xy - WaveOrigin;\r\nfloat  alongWave = dot(relPos, travelDir);\r\n\r\nfloat crestPos      = WaveSpeed * LocalTime;\r\nfloat distFromCrest = alongWave - crestPos;\r\n\r\nfloat widthBehind  = max(WaveWidth, 1e-3);\r\nfloat widthFront   = max(WaveWidth / max(FrontSharpness, 1.0), 1e-3);\r\nfloat falloffWidth = (distFromCrest < 0.0) ? widthBehind : widthFront;\r\nfloat profile      = exp(-(distFromCrest * distFromCrest) / (falloffWidth * falloffWidth));\r\n\r\nfloat height = profile * WaveHeight;\r\n\r\nfloat crestMask = saturate((profile - 0.4) / 0.6);\r\nfloat lean      = crestMask * crestMask * CrestLean * WaveHeight;\r\n\r\nfloat curlDist = distFromCrest - widthFront * 0.6;\r\nfloat curl     = exp(-(curlDist * curlDist) / (widthFront * widthFront * 0.35));\r\nfloat curlGate = smoothstep(-widthFront * 0.15, widthFront * 0.15, curlDist);\r\nfloat curlAdd  = curl * CurlStrength * WaveHeight * curlGate;\r\nheight += curlAdd;\r\nlean   += curl * CurlStrength * WaveHeight * 0.5;\r\n\r\nfloat t01       = saturate(LocalTime / max(LifeWindow, 1e-5));\r\nfloat lifeShape = sin(t01 * 3.14159265);\r\nfloat life      = (LifeWindow > 0.0) ? lifeShape : 1.0;\r\nheight *= life;\r\nlean   *= life;\r\n\r\nfloat finalVertexZ = LocalPos.z + height;\r\nfloat alpha        = saturate(finalVertexZ / max(WaveHeight, 1e-5));\r\n\r\nfloat displacedDistFromCrest = distFromCrest + lean;\r\nfloat displacedFalloff       = (displacedDistFromCrest < 0.0) ? widthBehind : widthFront;\r\nfloat displacedProfile       = exp(-(displacedDistFromCrest * displacedDistFromCrest)\r\n                                   / (displacedFalloff * displacedFalloff));\r\nCrestGradient = saturate((displacedProfile - 0.4) / 0.6) * life;\r\n\r\nfloat2 horizontalOffset = travelDir * lean;\r\nreturn float4(horizontalOffset.x, horizontalOffset.y, height, alpha);"
   description=""
   outputType=CMOT_Float4
   showCode=true
   additionalOutputs [{outputName=CrestGradient, outputType=CMOT_Float1}]
   inputs [
    {
     inputName=LocalPos
     input {
      expression=&self:MaterialExpressionLocalPosition_1
      mask=1
      maskB=1
      maskG=1
      maskR=1
     }
    }
    {
     inputName=LocalTime
     input {
      expression=&self:MaterialExpressionDynamicParameter_6
      mask=1
      maskR=1
     }
    }
    {
     inputName=WaveDir
     input {expression=&self:MaterialExpressionComponentMask_2}
    }
    {
     inputName=WaveOrigin
     input {expression=&self:MaterialExpressionAppendVector_2}
    }
    {
     inputName=WaveSpeed
     input {
      expression=&self:MaterialExpressionDynamicParameter_6
      mask=1
      maskG=1
      outputIndex=1
     }
    }
    {
     inputName=WaveHeight
     input {expression=&self:MaterialExpressionMultiply_38}
    }
    {
     inputName=WaveWidth
     input {
      expression=&self:MaterialExpressionDynamicParameter_6
      mask=1
      maskA=1
      outputIndex=3
     }
    }
    {
     inputName=FrontSharpness
     input {
      expression=&self:MaterialExpressionDynamicParameter_7
      mask=1
      maskR=1
     }
    }
    {
     inputName=CrestLean
     input {
      expression=&self:MaterialExpressionDynamicParameter_7
      mask=1
      maskG=1
      outputIndex=1
     }
    }
    {
     inputName=CurlStrength
     input {
      expression=&self:MaterialExpressionDynamicParameter_7
      mask=1
      maskB=1
      outputIndex=2
     }
    }
    {
     inputName=LifeWindow
     input {
      expression=&self:MaterialExpressionDynamicParameter_7
      mask=1
      maskA=1
      outputIndex=3
     }
    }
   ]
  }
  n43 = ComponentMask(None=@n42.return) {b=true, g=true, r=true}
  n44 = DynamicParameter() {parameterIndex=2, defaultValue=<rgba 0.2 1 1 1>, paramNames [Time, Wave Speed, Wave Hight, Wave Width]}
  n45 = Divide(
   A=ComponentMask(None=Add(A=@n39.XYZ, B=@n43)) {b=true}
   B=Multiply(A=@n44.Wave Hight, B=ScalarParameter() {defaultValue=1.9508, parameterName=Param_2})
  )
  n46 = ScalarParameter() {defaultValue=9.4473, parameterName=Param_6}
  n47 = ScalarParameter() {defaultValue=-0.7106, parameterName=Param_5}
  n48 = Add(A=LocalPosition().XYZ, B=@n43)
  n49 = Normalize(VectorInput=CrossProduct(A=DDX(Value=@n48), B=DDY(Value=@n48)))
  n50 = Fresnel(
   ExponentIn=ScalarParameter() {defaultValue=0.2639, parameterName=Param_3}
   Normal=NamedRerouteDeclaration(None=Transform(None=@n49)) {name=Normal, nodeColor=<rgba 0 0.3437 1 1>}
  ) {baseReflectFraction=0, exponent=1}
  n51 = Saturate(None=Multiply(A=OneMinus(None=@n45), B=LinearInterpolate(A=@n46, B=@n47, Alpha=@n50)))
  MaterialFunctionCall(Opacity In=Multiply(A=Multiply(A=@n38.A, B=@n51), B=ComponentMask(None=@n42.return) {a=true}), DepthFade Distance=DynamicParameter().Param3) {materialFunction=&MF_VCVFX_DepthFade}
  Multiply(B=@n42.CrestGradient)
  n52 = Add(A=@n15.R, B=@n18.G)
  n53 = VectorParameter() {parameterName=Color Texture mask Options, channelNames {a="", b=Add, g=Multiply, r=Power}, defaultValue=<rgba 1 1 0 0>}
  n54 = Multiply(A=Power(Base=Add(A=@n52, B=@n21.B), Exp=@n53.Power), B=@n53.Multiply)
  Multiply(A=Multiply(A=Add(A=@n54, B=@n53.Add), B=VertexNormalWS()), B=ScalarParameter() {defaultValue=110, parameterName=WPO Scale})
  n55 = DynamicParameter() {parameterIndex=1, paramNames [Shadow Exp, Param2, Param3, Param4]}
  NamedRerouteDeclaration(None=@n55.Shadow Exp) {name=WaveHight, nodeColor=<rgba 1 0 0.7486 1>}
  NamedRerouteDeclaration(None=@n55.Param2) {name=WaveBend, nodeColor=<rgba 1 0 0.7486 1>}
  NamedRerouteDeclaration(None=@n55.Param4) {name=WaveMask, nodeColor=<rgba 1 0 0.7486 1>}
  NamedRerouteDeclaration(None=@n55.Param3) {name=WaveBendPivot, nodeColor=<rgba 1 0 0.7486 1>}
  OneMinus(None=@n50)
  Tangent!(None=CrossProduct(A=VertexNormalWS(), B=VertexTangentWS()))
  Transform(None=@n43) {transformSourceType=TRANSFORMSOURCE_Local}
  n56 = Normalize(VectorInput=@n48)
  VertexInterpolator(VS=Normalize(VectorInput=Transform(None=@n56) {transformSourceType=TRANSFORMSOURCE_Local, transformType=TRANSFORM_Tangent}))
 ]
}
