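# Masked, unlit stylized ocean wave material for Niagara mesh particles in the Dealership library's RoutineBarley branding.
# Three panning caustic samples combine by minimum into a pseudo-normal that drives displacement, fake two-color lighting and specular highlight.
# A particle-driven sphere mask adds a sine swell and bounds the water.
# Noise erosion and dissolve feed dithered opacity, while swirl-texture foam gathers on wave and eroded edges.
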Material /DealershipSharedLibrary/Assets/ThirdPartyBranding/RoutineBarley/M_DLR_RoutineBarly_Ocean
graph {
 properties {
  bAutomaticallySetUsageInEditor=false
  bUsedWithNiagaraMeshParticles=true
  bUsedWithStaticMesh=false
  blendMode=BLEND_Masked
  shadingModel=MSM_Unlit
  usedShadingModels=Unlit
 }
 nodes [
  n1 = Multiply(A=TextureCoordinate(), B=ScalarParameter() {defaultValue=2.8917, parameterName=Color Variation Tile})
  n2 = VectorParameter() {parameterName=Param_2, defaultValue=<rgba 0.172 0.1665 0 0.776>}
  n3 = TextureObjectParameter() {parameterName=Color Noise Texture, texture=/Game/Global/Textures/Default/Noise/T_SoftNoise}
  n4 = TextureSample(UVs=Panner(Coordinate=@n1, Speed=ComponentMask(None=@n2.RGB) {g=true, r=true}), Tex=@n3) {texture=&T_ShieldCaustics}
  Add(A=Add(A=@n4.RGB))
  n5 = VectorParameter() {parameterName=Param_1, defaultValue=<rgba 0.4181 0.3548 -0.064 0.08>}
  n6 = ComponentMask(None=@n5.RGB) {g=true, r=true}
  n7 = Multiply(
   A=Panner(Coordinate=Add(A=@n1, B=@n6), Speed=ComponentMask(None=@n5.RGBA) {a=true, b=true})
   B=ScalarParameter() {defaultValue=1.5117, parameterName=B}
  ) {constB=0.6183}
  n8 = VectorParameter() {parameterName=Param_3, defaultValue=<rgba 0.8649 0.1484 -0.12 0.04>}
  n9 = Panner(Coordinate=Add(A=@n1, B=ComponentMask(None=@n8.RGB) {g=true, r=true}), Speed=ComponentMask(None=@n8.RGBA) {a=true, b=true})
  n10 = ScalarParameter() {defaultValue=1.8711, parameterName=B_1}
  n11 = Min(
   A=Min(A=@n4.RGB, B=TextureSample(UVs=@n7, Tex=@n3) {texture=&T_ShieldCaustics}.RGB)
   B=TextureSample(UVs=Multiply(A=@n9, B=@n10) {constB=0.8088}, Tex=@n3) {texture=&T_ShieldCaustics}.RGB
  )
  n12 = Normalize(VectorInput=@n11)
  n13 = Multiply(A=@n12, B=VertexColor().R)
  n14 = VertexNormalWS()
  n15 = DynamicParameter() {paramNames [SphereMaskRadius, SphereMaskContrast, MaskOffset X, WaveTime]}
  n16 = SphereMask(A=TextureCoordinate(), B=AppendVector(A=@n15.MaskOffset X, B=Constant() {r=0.5}), Radius=@n15.SphereMaskRadius) {attenuationRadius=0.5, hardnessPercent=0}
  n17 = Add(A=Multiply(A=@n16, B=ScalarParameter() {defaultValue=2, parameterName=Param_1}), B=@n15.WaveTime)
  n18 = Multiply(A=@n16, B=ConstantBiasScale(None=Sine(None=@n17)))
  Add(
   A=Multiply(A=Multiply(A=@n13, B=@n14), B=ScalarParameter() {defaultValue=150, parameterName=WPO Scale})
   B=Multiply(A=@n18, B=VectorParameter() {defaultValue=<rgba 0 0 100 1>}.RGB)
  )
  n19 = ComponentMask(None=@n11) {g=true, r=true}
  n20 = VectorParameter() {parameterName=Color Texture mask Options, channelNames {a="", b=Add, g=Multiply, r=Power}, defaultValue=<rgba 0.5536 0.5136 0 0>}
  n21 = Add(A=Multiply(A=Power(Base=@n19, Exp=@n20.Power), B=@n20.Multiply), B=@n20.Add)
  n22 = MaterialFunctionCall!(
   Dynamic Param=@n21
   Alpha=MaterialFunctionCall(In=@n16, Contrast=@n15.SphereMaskContrast) {materialFunction=&CheapContrast}.Result
  ) {materialFunction=/Game/Effects/Fort_Effects/Materials/_Functions/Erode_Alpha}
  n23 = DynamicParameter() {parameterIndex=1, defaultValue=<rgba 1 0 1 1>, paramNames [ShadowExp,Dissolve,Param3,Param4]}
  n24 = ComponentMask(None=@n12) {b=true}
  n25 = ScalarParameter() {parameterName="Edge Width (S)"}
  n26 = ScalarParameter() {parameterName="Width (S)"}
  MaterialFunctionCall(Alpha Threshold=Multiply(
    A=@n22.Result
    B=MaterialFunctionCall!(Erosion=@n23.Dissolve, Mask=@n24, Edge Width=@n25, Width=@n26) {materialFunction=/DealershipSharedLibrary/FX/Materials/MaterialFunctions/MF_DLR_Erode}.Result
   )) {materialFunction=&DitherTemporalAA}
  n27 = VectorParameter() {parameterName=Param_6, defaultValue=<rgba 0.4523 0 1 1>}
  n28 = Multiply(
   A=SphereMask(
    A=ReflectionVectorWS(CustomWorldNormal=@n12)
    B=Transform(None=@n27.RGB) {transformSourceType=TRANSFORMSOURCE_Local}
    Radius=ScalarParameter() {defaultValue=0.15, parameterName=Param_4}
    Hardness=ScalarParameter() {defaultValue=0.2, parameterName=Param_9}
   )
   B=ScalarParameter() {defaultValue=4, parameterName=Param_12}
  ) {constB=2}
  n29 = VectorParameter() {parameterName=Param_5, defaultValue=<rgba 0.488 0 1 1>}
  n30 = DotProduct(A=@n12, B=Transform(None=Normalize(VectorInput=@n29.RGB)) {transformSourceType=TRANSFORMSOURCE_Local})
  n31 = LinearInterpolate(
   A=VectorParameter() {parameterName=Color_01, defaultValue=<rgba 0.086 0.2358 0.401 1>}.RGB
   B=VectorParameter() {parameterName=Color_02, defaultValue=<rgba 0.0355 0.135 0.401 1>}.RGB
   Alpha=Power(Base=ConstantBiasScale(None=@n30), Exp=@n23.ShadowExp)
  )
  n32 = VectorParameter() {parameterName=FoamColor, defaultValue=<rgba 0.9792 0.9792 0.9792 1>}
  n33 = ScalarParameter() {defaultValue=0.1, parameterName=Param_3}
  n34 = Multiply(A=Add(A=TextureCoordinate(), B=Multiply(A=@n21, B=@n33)), B=ScalarParameter() {defaultValue=2.0994, parameterName=Caustic_Scale_X})
  n35 = VertexColor()
  n36 = OneMinus(None=@n22.Result)
  n37 = Multiply(
   A=Multiply(
    A=TextureSample(UVs=@n34) {samplerType=SAMPLERTYPE_LinearColor, texture=/Game/Global/Textures/Default/Caustics/WaterFoam_AccentSwirls_H}.R
    B=ScalarParameter() {defaultValue=3.4237, parameterName=Foam Amount}
   )
   B=Add(A=Multiply(A=@n35.R, B=@n18), B=@n36)
  )
  n38 = ScalarParameter() {defaultValue=3, parameterName=Param_8}
  n39 = Multiply(A=@n36) {constB=0.5}
  n40 = Saturate(None=Add(A=Power(Base=@n37, Exp=@n38), B=Multiply(A=@n39, B=@n39)))
  MaterialFunctionCall(Emissive=Add(A=Multiply(A=@n28, B=@n28), B=LinearInterpolate(A=@n31, B=@n32.RGB, Alpha=@n40))) {materialFunction=&MF_FX_EmissiveAdjustments}
  Multiply(
   A=MaterialFunctionCall() {materialFunction=&DitherTemporalAA}.Result
   B=ScalarParameter() {defaultValue=20}
  )
 ]
}
